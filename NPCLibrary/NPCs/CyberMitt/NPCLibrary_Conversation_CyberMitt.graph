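# Conversation database for the CyberMitt NPC in Fortnite, providing its greeting and the branches that follow.
# The entry tag leads to one speech node with a default line plus contextual lines gated by first-ever conversation, per-match conversation count or player character cosmetic.
# Priorities, some curve-driven, weight those lines.
# The speech node then fans out through link nodes to shared quest and basic-service entries and to CyberMitt's own dialogue entry.
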ConversationDatabase /NPCLibrary/NPCs/CyberMitt/NPCLibrary_Conversation_CyberMitt
entry_points [{tag=Conversation.Entry.CyberMitt, destinations [#1]}]
nodes [
 ConversationEntryPointNode_7 {
  id=#1
  class=ConversationEntryPointNode
  outputs [#2]
  properties {entryTag {tagName=Conversation.Entry.CyberMitt}}
 }
 Speech_C_4 {
  id=#2
  class=Speech_C
  outputs [#3, #4, #5]
  properties {
   nodeName=""
   generalConfig {
    defaultMessage=What use is honor if it paralyzes you? I'm ready to FIGHT.
    contextualMessages [
     {
      message=My blade's at your service... maybe.
      contextRequirements [{requirement {class=FortControllerRequirement_IsFirstEverConversationWithNPC}}]
      priority {curve {rowName=Tandem.Speech.Priority.FirstEverInteraction}}
     }
     {
      message=I'm sick of waiting for my destiny. I'm seizing it.
      contextRequirements [{requirement {class=FortControllerRequirement_CheckNPCMatchConversationCount, properties {nPCUniqueIdTag {tagName=AISpawnerData.Type.Tandem.CyberMitt}}}}]
      priority {curve {curveTable=null, rowName=None}}
     }
     {
      message=What use is honor if it paralyzes you? I'm ready to FIGHT.
      contextRequirements [{requirement {class=FortControllerRequirement_CheckNPCMatchConversationCount, properties {nPCUniqueIdTag {tagName=AISpawnerData.Type.Tandem.CyberMitt}}}}]
      priority {curve {curveTable=null, rowName=None}}
     }
     {
      message="This Island belongs to all of us... We all have to step up and defend it!"
      contextRequirements [{requirement {class=FortControllerRequirement_CheckNPCMatchConversationCount, properties {nPCUniqueIdTag {tagName=AISpawnerData.Type.Tandem.CyberMitt}}}}]
      priority {curve {curveTable=null, rowName=None}}
     }
     {
      message="Stand back, brother. Let a warrior show you how it's done."
      contextRequirements [{requirement {
         class=FortControllerRequirement_HasCID
         properties {allowedCIds [&Character_MistMylar_Claw, /BRCosmetics/Athena/Items/Cosmetics/Characters/Character_SolidMist]}
        }}]
      priority {value=2, curve {curveTable=null, rowName=None}}
     }
     {
      message=I don't have magic. I'm not 'special'. But I know how to stand up for what's right.
      contextRequirements [{requirement {
         class=FortControllerRequirement_HasCID
         properties {allowedCIds [/BRCosmetics/Athena/Items/Cosmetics/Characters/Character_PolishedJade_Mind]}
        }}]
      priority {value=2, curve {curveTable=null, rowName=None}}
     }
     {
      message=I know our... friendship... has an expiration date. Just don't forget me when you leave.
      contextRequirements [{requirement {
         class=FortControllerRequirement_HasCID
         properties {allowedCIds [
           &Character_IronLilac_Sly
           &Character_BrightDisk
           &Character_VoidRedemption_Rebel
          ]}
        }}]
      priority {value=2, curve {curveTable=null, rowName=None}}
     }
     {
      message="The Bugs are gone, but I can feel it in my gut... Something worse is coming."
      contextRequirements [{requirement {class=FortControllerRequirement_HasCID, properties {allowedCIds [/BRCosmetics/Athena/Items/Cosmetics/Characters/Character_SwoopClaspPry]}}}]
      priority {value=2, curve {curveTable=null, rowName=None}}
     }
    ]
   }
  }
 }
 ConversationLinkNode_2 {id=#5, class=ConversationLinkNode, properties {remoteEntryTag {tagName=Conversation.Entry.CyberMitt.Dialogue}}}
 ConversationLinkNode_0 {id=#4, class=ConversationLinkNode, properties {remoteEntryTag {tagName=Conversation.Entry.BasicServices}}}
 ConversationLinkNode_1 {id=#3, class=ConversationLinkNode, properties {remoteEntryTag {tagName=Conversation.Entry.NPCLibrary.Quests}}}
]
