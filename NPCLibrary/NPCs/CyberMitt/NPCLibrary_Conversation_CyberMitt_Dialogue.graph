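# Conversation database for the CyberMitt NPC's "Ask about..." dialogue, entered via the Conversation.Entry.CyberMitt.Dialogue tag.
# A greeting speech node branches into player choices for Purpose, The Spirit Realm and Daigo, plus a Back option.
# Each topic plays a linear chain of speech lines about CyberMitt's outlook and brother Daigo.
# All three chains converge on one shared ReturnToStart node labelled BACK.
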ConversationDatabase /NPCLibrary/NPCs/CyberMitt/NPCLibrary_Conversation_CyberMitt_Dialogue
entry_points [{tag=Conversation.Entry.CyberMitt.Dialogue, destinations [#1]}]
nodes [
 ConversationEntryPointNode_7 {
  id=#1
  class=ConversationEntryPointNode
  outputs [#2]
  properties {entryTag {tagName=Conversation.Entry.CyberMitt.Dialogue}}
 }
 Speech_C_11 {
  id=#2
  class=Speech_C
  outputs [#3, #4, #5, #6]
  properties {nodeName="", softTaskIcon=&T_UI_NPC_question, generalConfig {defaultMessage=How can I help?}}
  sub_nodes [ConversationChoiceNode_5 {class=ConversationChoiceNode, properties {defaultChoiceDisplayText=Ask about...}}]
 }
 Back_C_0 {
  id=#6
  class=Back_C
  sub_nodes [ConversationChoiceNode_7 {class=ConversationChoiceNode, properties {defaultChoiceDisplayText=BACK}}]
 }
 Speech_C_6 {
  id=#5
  class=Speech_C
  outputs [#7]
  properties {
   nodeName=""
   softTaskIcon=&T_UI_NPC_question
   generalConfig {defaultMessage=Daigo thinks I'm still angry with him for the things he's done.}
  }
  sub_nodes [ConversationChoiceNode_15 {class=ConversationChoiceNode, properties {defaultChoiceDisplayText=Daigo}}]
 }
 Speech_C_8 {
  id=#7
  class=Speech_C
  outputs [#8]
  properties {
   nodeName=""
   softTaskIcon=&T_UI_NPC_question
   generalConfig {defaultMessage="I'm not. I accept him, in all his flawed glory. My clever, reckless brother."}
  }
 }
 Speech_C_9 {
  id=#8
  class=Speech_C
  outputs [#9]
  properties {
   nodeName=""
   softTaskIcon=&T_UI_NPC_question
   generalConfig {defaultMessage=But we don't have time for looking back. For seeking redemption.}
  }
 }
 Speech_C_10 {
  id=#9
  class=Speech_C
  outputs [#10]
  properties {
   nodeName=""
   softTaskIcon=&T_UI_NPC_question
   generalConfig {defaultMessage="If he wants to help, he can pick up a weapon and fight. Like anyone else."}
  }
 }
 ReturnToStart_C_23 {
  id=#10
  class=ReturnToStart_C
  properties {softTaskIcon=&T_UI_NPC_Back}
  sub_nodes [ConversationChoiceNode_32 {class=ConversationChoiceNode, properties {defaultChoiceDisplayText=BACK}}]
 }
 Speech_C_2 {
  id=#4
  class=Speech_C
  outputs [#11]
  properties {
   nodeName=""
   softTaskIcon=&T_UI_NPC_question
   generalConfig {defaultMessage=I once dreamed with my whole heart of seeing magic someday.}
  }
  sub_nodes [ConversationChoiceNode_1 {class=ConversationChoiceNode, properties {defaultChoiceDisplayText=The Spirit Realm}}]
 }
 Speech_C_3 {
  id=#11
  class=Speech_C
  outputs [#10]
  properties {
   nodeName=""
   softTaskIcon=&T_UI_NPC_question
   generalConfig {defaultMessage=I'm tired of dreaming. It's time to wake up and fight.}
  }
 }
 Speech_C_7 {
  id=#3
  class=Speech_C
  outputs [#12]
  properties {
   nodeName=""
   softTaskIcon=&T_UI_NPC_question
   generalConfig {defaultMessage=I've had my head in the clouds for a long time. Not anymore.}
  }
  sub_nodes [ConversationChoiceNode_0 {class=ConversationChoiceNode, properties {defaultChoiceDisplayText=Purpose}}]
 }
 Speech_C_15 {
  id=#12
  class=Speech_C
  outputs [#13]
  properties {
   nodeName=""
   softTaskIcon=&T_UI_NPC_question
   generalConfig {defaultMessage="I decide who I am. Not my siblings, not stories of the old masters, not fate. ME."}
  }
 }
 Speech_C_14 {
  id=#13
  class=Speech_C
  outputs [#14]
  properties {
   nodeName=""
   softTaskIcon=&T_UI_NPC_question
   generalConfig {defaultMessage="And if I'm a little wilder, a little rougher, a little freer than I was before..."}
  }
 }
 Speech_C_0 {
  id=#14
  class=Speech_C
  outputs [#10]
  properties {
   nodeName=""
   softTaskIcon=&T_UI_NPC_question
   generalConfig {defaultMessage=I think I've earned the chance to let loose. Don't you?}
  }
 }
]
